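# Additive unlit Niagara sprite material for a crystal glow in the Season 21 Battle Pass collectable coin assets.
# A powered radial mask forms core and edge, combined with a flare texture and tinted by particle or rainbow color, with optional hue shift.
# Layered sines drive flicker, multiplied with depth fade and optional angle and near-camera fades.
# The Sprite function computes vertex offsets around a sine-bobbing object pivot, sized by dynamic parameters.
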Material /BattlePassS21/QuestActors/Collectable_Coin/M_Sprite_Crystal_Glow1
graph {
 properties {bUsedWithNiagaraSprites=true, blendMode=BLEND_Additive, pixelDepthOffsetMode=PDOM_Legacy, shadingModel=MSM_Unlit, usedShadingModels=Unlit}
 nodes [
  n1 = MaterialFunctionCall() {materialFunction=&ObjectPivotPoint}
  n2 = Time()
  n3 = ScalarParameter() {defaultValue=0.4, parameterName=WPOTimeScale}
  n4 = LinearInterpolate(
   A=VectorParameter() {parameterName=BobHeightMax, defaultValue=<rgba 0 0 10 0>}.RGB
   B=VectorParameter() {parameterName=BobHeightMin, defaultValue=<rgba 0 0 -2 0>}.RGB
   Alpha=Sine(None=Multiply(A=@n2, B=@n3))
  )
  n5 = DynamicParameter() {desc=Index 1 Params for Niagara, parameterIndex=1, paramNames [DepthFadeDist, X_Sprite, Y_Sprite, Param4]}
  n6 = Multiply(A=@n5.X_Sprite, B=Constant() {r=75})
  n7 = Multiply(A=@n5.Y_Sprite, B=Constant() {r=150})
  Add(
   A=VectorParameter() {parameterName=Height}.RGB
   B=MaterialFunctionCall!(Center Location=Add(A=@n1.Object Pivot Location, B=@n4), 0-1 UVs, XY scale=AppendVector(A=@n6, B=@n7), OverridePivot, "Normals (Optional)") {materialFunction=&Sprite}.World position offset
  )
  Constant() {r=1}
  n8 = ParticleRandom()
  n9 = DynamicParameter() {paramNames [FlickerFade,FlickerSpeed,ColorJitter,Param4]}
  n10 = Multiply(
   A=Time()
   B=StaticSwitchParameter(True=LinearInterpolate(Alpha=@n8) {constA=1.2, constB=0.8}, False=@n9.FlickerSpeed) {parameterName=GPU Particles}
  )
  n11 = Sine(None=@n10) {period=0.075}
  n12 = Multiply(A=Multiply(A=Sine(None=@n10) {period=0.155}, B=Sine(None=@n10) {period=0.2}), B=Sine(None=@n10) {period=0.43})
  n13 = ScalarParameter() {defaultValue=0.4, desc="0 = full flicker amount", parameterName=GPUFlickerFadeMin}
  n14 = ScalarParameter() {defaultValue=1, desc="1 = no flicker", parameterName=GPUFlickerFadeMax, sortPriority=1}
  n15 = ParticleRandom()
  n16 = LinearInterpolate(
   A=Clamp(None=Multiply(A=@n11, B=@n12)) {maxDefault=3}
   Alpha=StaticSwitchParameter(True=LinearInterpolate(A=@n13, B=@n14, Alpha=@n15) {constA=0.5}, False=Saturate(None=@n9.FlickerFade)) {parameterName=GPU Particles}
  ) {constA=1}
  n17 = Constant() {r=0.8}
  FeatureLevelSwitch(Default=@n16, ES2=@n17, ES3_1=@n17, SM4_REMOVED=@n16)
  n18 = ScalarParameter() {defaultValue=2, parameterName=CenterBrightness}
  n19 = Constant() {r=0.5}
  n20 = SphereMask(A=TextureCoordinate(), B=@n19, Radius=@n19) {hardnessPercent=0}
  n21 = Power(Base=@n20, Exp=ScalarParameter() {defaultValue=25, parameterName=CentertPower}) {constExponent=25}
  n22 = ComponentMask(None=TextureCoordinate() {uTiling=0, vTiling=0}) {g=true}
  n23 = Multiply(
   A=TextureSample(UVs=Panner(Coordinate=@n22) {speedX=1.6667}) {texture=/Game/Global/Textures/Default/Masks_and_Gradients/T_Grad_Rainbow}.RGB
   B=ParticleColor().RGB
  )
  n24 = ParticleColor()
  n25 = StaticSwitchParameter(
   True=MaterialFunctionCall(LDR Input=Clamp(None=@n23), HDR Tint=@n23) {materialFunction=&DeriveHDRfromLDR}.RemappedRange
   False=Desaturation(None=@n24.RGB, Fraction=ScalarParameter() {parameterName=inputDesaturation})
  ) {parameterName=Use Rainbow Color}
  n26 = ScalarParameter() {defaultValue=0.3, parameterName=CenterDesaturation}
  n27 = Multiply(
   A=Power(Base=@n20, Exp=ScalarParameter() {defaultValue=4, parameterName=EdgePower}) {constExponent=4}
   B=ScalarParameter() {defaultValue=2.5, parameterName=EdgeBrightness}
  ) {constB=0.15}
  n28 = TextureSampleParameter2D() {parameterName=FlareMask, texture=/Game/Effects/Fort_Effects/Textures/Flares/T_Flare_Streaks}
  n29 = Add(A=Multiply(A=Multiply(A=@n18, B=@n21) {constB=2}, B=Desaturation(None=@n25, Fraction=@n26)), B=Multiply(A=Add(A=@n27, B=@n28.RGB), B=@n25))
  MaterialFunctionCall(Emissive=StaticSwitchParameter(
    True=MaterialFunctionCall(Hue Shift Percentage=@n9.ColorJitter, Texture=@n29) {materialFunction=&HueShift}.Result
    False=@n29
   ) {parameterName=UseColorJitter?}) {materialFunction=&MF_FX_EmissiveAdjustments}
  n30 = MaterialFunctionCall("FadeSensitivity (optional)"=ScalarParameter() {defaultValue=1, parameterName=AngleFade}) {materialFunction=&FadeBasedOnViewAngle}
  n31 = Constant() {r=1}
  n32 = Constant() {r=1}
  n33 = ScalarParameter() {defaultValue=30, parameterName=DepthFade}
  n34 = DepthFade(Opacity=@n24.A, FadeDistance=If(A=@n5.DepthFadeDist, B=Constant() {r=1}, "A > B"=@n5.DepthFadeDist, "A == B"=@n33, "A < B"=@n33))
  n35 = Multiply(
   A=Multiply(
    A=StaticSwitchParameter(True=@n30.Result, False=@n31) {parameterName=Use Angle Fade}
    B=StaticSwitchParameter(True=@n16, False=@n32) {defaultValue=true, parameterName=UseFlicker}
   )
   B=FeatureLevelSwitch(Default=@n34, ES2=@n24.A, ES3_1=@n24.A, SM4_REMOVED=@n34, SM5=@n34, SM6=@n34)
  )
  n36 = ScalarParameter() {parameterName=CameraFadeDistance}
  StaticSwitchParameter(
   True=Multiply(A=@n35, B=MaterialFunctionCall(Distance=@n36) {materialFunction=&Near_Camera_Fade}.Result)
   False=@n35
  ) {group=DetailedGradientMask, parameterName=UseCameraNearFade}
 ]
}
